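# Translucent, two-sided unlit effect material for the DonutPlate1H pickaxe, producing an animated, three-color flowing noise look.
# Panning blurred and cellular noise plus sine waves distort a UV-based lookup into a curve atlas row, whose channels drive color blending and opacity.
# Opacity combines a tip mask, view-angle edge fade and depth fade, which mobile skips.
# Switch-style scalar lerps select optional dark globs, tip mask type and SplineThicken offset.
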Material /BRCosmetics/Weapons/FORT_Melee/Pickaxe_DonutPlate1H/FX/M_DonutPlate1H
graph {
 properties {
  bUsedWithMeshDeformer=true
  bUsedWithSkeletalMesh=true
  blendMode=BLEND_Translucent
  pixelDepthOffsetMode=PDOM_Legacy
  shadingModel=MSM_Unlit
  twoSided=true
  usedShadingModels=Unlit
 }
 nodes [
  AppendVector(A=Constant() {r=0.5}, B=Constant() {r=1})
  AppendVector(A=Constant() {r=0.5}, B=Constant() {r=1})
  Constant() {r=0.5}
  Constant() {r=1}
  Constant() {r=0.5}
  Constant() {r=0.7}
  LinearInterpolate(
   B=MaterialFunctionCall(WidthBase=ScalarParameter() {parameterName=WidthBase}, WidthTip=ScalarParameter() {parameterName=WidthTip}) {materialFunction=&SplineThicken}.WorldPositionOffset
   Alpha=ScalarParameter() {group=Switch, parameterName=SplineThicken}
  )
  n1 = MaterialFunctionCall(Float2=TextureCoordinate()) {materialFunction=&BreakOutFloat2Components}
  n2 = Multiply(A=Subtract(A=@n1.G) {constB=0.5}) {constB=2}
  n3 = LinearInterpolate(
   A=@n1.G
   B=Multiply(A=Abs(None=@n2), B=ScalarParameter() {defaultValue=1, parameterName=Stretch})
   Alpha=ScalarParameter() {parameterName=FloatingType}
  )
  n4 = Constant() {r=-0.2}
  n5 = LinearInterpolate(A=@n4, B=OneMinus(None=@n4), Alpha=@n3) {constA=-1, constB=2}
  n6 = Add(A=VertexColor().R, B=AppendVector(A=@n1.R, B=LinearInterpolate(A=@n3, B=@n5, Alpha=@n1.R)))
  n7 = VectorParameter() {group="Texture_Ctrl(Tile&&Speed)", parameterName=T_Source_crl, defaultValue=<rgba 0.5 1 0 0.6>}
  n8 = AppendVector(A=@n7.R, B=@n7.R)
  n9 = TextureSample(UVs=Panner(Coordinate=Multiply(A=@n6, B=@n8), Speed=AppendVector(A=@n7.B, B=@n7.B)) {speedY=0.6}) {samplerType=SAMPLERTYPE_LinearGrayscale, texture=/Engine/Functions/Engine_MaterialFunctions02/ExampleContent/Textures/LowResBlurredNoise}
  n10 = LinearInterpolate(Alpha=@n9.R) {constA=-1}
  n11 = Multiply(B=ScalarParameter() {defaultValue=1, parameterName=Distortion_Intensity}) {constA=0.1}
  n12 = TextureCoordinate()
  n13 = Time()
  n14 = Multiply(A=Add(A=ComponentMask(None=@n12) {g=true}, B=Multiply(A=@n13) {constB=0.2})) {constB=2}
  n15 = ScalarParameter() {defaultValue=1, parameterName=Distortion_Intensity}
  n16 = Add(
   A=Add(A=Multiply(A=@n10, B=@n11) {constB=0.1}, B=ComponentMask(None=@n12) {r=true})
   B=Multiply(A=Sine(None=@n14), B=Multiply(B=@n15) {constA=0.05}) {constB=0.05}
  )
  n17 = ComponentMask(None=TextureCoordinate()) {g=true}
  n18 = Add(A=Multiply(A=@n17) {constB=4}, B=Multiply(A=Time()) {constB=0.5})
  n19 = Multiply(A=Add(A=Sine(None=@n18))) {constB=0.3}
  n20 = VectorParameter() {group="Texture_Ctrl(Tile&&Speed)", parameterName=T_Source_crl, defaultValue=<rgba 0.5 1 0 0.6>}
  n21 = AppendVector(A=@n20.R, B=@n20.R)
  n22 = VectorParameter() {group="Texture_Ctrl(Tile&&Speed)", parameterName=T_Distort_crl, defaultValue=<rgba 0.5 0.7 -0.3 0.6>}
  n23 = AppendVector(A=@n22.R, B=@n22.R)
  n24 = TextureSample(UVs=Panner(Coordinate=Multiply(A=@n6, B=@n23), Speed=AppendVector(A=@n22.B, B=@n22.B)) {speedX=-0.3, speedY=0.6}) {samplerType=SAMPLERTYPE_Grayscale, texture=/Game/Global/Textures/Default/Noise/LowResBlurredNoise}
  n25 = Add(A=Panner(Coordinate=Multiply(A=@n6, B=@n21), Speed=AppendVector(A=@n20.B, B=@n20.B)) {speedY=0.6}, B=Multiply(A=@n24.R) {constB=0.3})
  n26 = Multiply(
   A=Subtract(
    A=Multiply(A=@n17, B=@n19)
    B=TextureSample(UVs=@n25) {samplerType=SAMPLERTYPE_Masks, texture=/Game/Global/Textures/Default/Noise/T_CellularNoise}.R
   )
   B=Multiply(B=ScalarParameter() {defaultValue=1, parameterName=Distortion_Intensity}) {constA=0.3}
  ) {constB=0.3}
  n27 = CurveAtlasRowParameter(CurveTime=Saturate(None=Add(A=@n16, B=@n26))) {
   atlas=/Game/Athena/LinearColorCurves/TempAthenaCurveAtlas
   curve=/Game/Weapons/FORT_Melee/Pickaxe_DonutPlate1H/FX/Curve_DonutPlate1H
   defaultValue=7
  }
  n28 = Multiply(
   A=LinearInterpolate(
    A=VectorParameter() {parameterName=Color_B, defaultValue=<rgba 1 0.1808 0.8445 1>}.RGB
    B=VectorParameter() {parameterName=Color_A, defaultValue=<rgba 1 1 1 1>}.RGB
    Alpha=@n27.R
   )
   B=@n27.G
  )
  n29 = LinearInterpolate(A=VectorParameter() {parameterName=Color_C, defaultValue=<rgba 0.1875 0 0.1519 1>}.RGB, B=@n28, Alpha=@n28)
  n30 = ComponentMask(None=TextureCoordinate()) {g=true}
  n31 = CurveAtlasRowParameter(CurveTime=@n30) {
   atlas=&TempAthenaCurveAtlas
   curve=&Curve_DonutPlate1H
   defaultValue=7
  }
  n32 = Sine(None=Divide(A=@n30))
  n33 = ScalarParameter() {group=Switch, parameterName=Tip_Opacity_Type}
  n34 = DotProduct(A=Normalize(VectorInput=VertexNormalWS()), B=CameraVectorWS())
  n35 = Multiply(
   A=Multiply(A=@n27.B, B=LinearInterpolate(A=@n31.A, B=@n32, Alpha=@n33))
   B=Power(Base=Abs(None=@n34), Exp=ScalarParameter() {defaultValue=1, parameterName=Fresnel_Power})
  )
  n36 = VectorParameter() {group="Texture_Ctrl(Tile&&Speed)", parameterName=T_Black_Glob_crl, defaultValue=<rgba 0.5 1 0.1 0.6>}
  n37 = Panner(Coordinate=Multiply(A=@n6, B=AppendVector(A=@n36.R, B=@n36.R)), Speed=AppendVector(A=@n36.B, B=@n36.B)) {speedX=0.1, speedY=0.6}
  n38 = Multiply(A=@n24.R) {constB=0.1}
  n39 = LinearInterpolate(Alpha=TextureSample(UVs=Add(A=@n37, B=@n38)) {samplerType=SAMPLERTYPE_LinearColor, texture=&HighContrastNoise}.R) {constA=-5, constB=3}
  n40 = LinearInterpolate(Alpha=Multiply(A=@n35, B=Saturate(None=@n39))) {constA=-1, constB=2}
  n41 = OneMinus(None=Multiply(A=Saturate(None=@n40), B=ScalarParameter() {parameterName=Dark_Glob_Intensity}))
  n42 = Multiply(
   A=LinearInterpolate(A=@n29, B=Multiply(A=@n29, B=@n41), Alpha=ScalarParameter() {group=Switch, parameterName=Dark_Glodbs_Switch})
   B=ScalarParameter() {defaultValue=1, parameterName=Emissive_Intensity}
  )
  MaterialFunctionCall(Emissive=@n42) {materialFunction=&MF_FX_EmissiveAdjustments}
  Reroute(None=@n42)
  n43 = Multiply(A=@n35, B=DepthFade(FadeDistance=ScalarParameter() {defaultValue=10, parameterName=Depth Fade Distance}))
  ShadingPathSwitch(Default=@n43, Deferred=@n43, Mobile=@n35) {desc=Migrated from FeatureLevelSwitch}
 ]
}
